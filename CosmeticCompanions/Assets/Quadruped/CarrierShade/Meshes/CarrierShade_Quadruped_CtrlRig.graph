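# Control Rig for the CarrierShade cosmetic quadruped companion that adjusts its Crown bone on the neck.
# The executed graph is a single aim constraint pointing Crown's Z axis at sibling bone Crown1, which also supplies the world-up direction space.
# A spherical pose reader on neck_02 that weights a Crown rotation, and a separate Crown translation offset, sit unconnected to the execution start.
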ControlRigBlueprint /CosmeticCompanions/Assets/Quadruped/CarrierShade/Meshes/CarrierShade_Quadruped_CtrlRig
graphs [
 RigVMFunctionLibrary {links [], nodes []}
 RigVMModel {nodes [
   RigUnit_BeginExecution = &RigUnit_BeginExecution()
   &RigUnit_AimConstraintLocalSpaceOffset(
    AimAxis=(X=0,Y=0,Z=1)
    Child=(Type=Bone,Name="Crown")
    Filter=(bX=True,bY=True,bZ=True)
    Parents=((Item=(Type=Bone,Name="Crown1"),Weight=1))
    UpAxis=(X=0,Y=1,Z=0)
    Weight=1
    WorldUp=(Target=(X=0,Y=1,Z=0),Kind=Direction,Space=(Type=Bone,Name="Crown1"))
    ExecutePin=@RigUnit_BeginExecution.ExecutePin
   )
   n1 = &RigUnit_SphericalPoseReader(
    ActiveRegionScaleFactors=(PositiveWidth=0.75,NegativeWidth=0.75,PositiveHeight=0.75,NegativeHeight=0.75)
    ActiveRegionSize=0.1
    Debug=(bDrawDebug=true,bDraw2D=False,bDrawLocalAxes=False,DebugScale=25,DebugSegments=20,DebugThickness=0.25)
    DriverAxis=(X=1,Y=0,Z=0)
    DriverItem=(Type=Bone,Name="neck_02")
    ExecutePin=()
    FalloffRegionScaleFactors=(PositiveWidth=1,NegativeWidth=1,PositiveHeight=1,NegativeHeight=1)
    FalloffSize=0.2
    FlipHeightScaling=False
    FlipWidthScaling=False
    OptionalParentItem=(Type=Bone,Name="None")
    RotationOffset=(X=0,Y=-90,Z=40)
   )
   > &RigUnit_ModifyTransforms(
    ItemToModify=((Item=(Type=Bone,Name="Crown"),Transform=(Rotation=(X=0.1736,Y=0,Z=0,W=0.9848),Translation=(X=0,Y=0,Z=0),Scale3D=(X=1,Y=1,Z=1))))
    WeightMinimum=0
    Weight=@n1.OutputParam
   )
   &RigUnit_ModifyTransforms(
    ExecutePin=()
    ItemToModify=((Item=(Type=Bone,Name="Crown"),Transform=(Rotation=(X=0,Y=0,Z=0,W=1),Translation=(X=0,Y=0.5,Z=0),Scale3D=(X=1,Y=1,Z=1))))
    Weight=1
    WeightMinimum=0
   )
  ]}
]
hierarchy [
 root {loc=<0 0 0>, rot=<0 0 0 1>}
 pelvis {parent=root, loc=<0 -14.664 26.7066>, rot=<0.5008 0.4992 -0.5008 -0.4992>}
 spine_01 {parent=pelvis, loc=<0 0 0>, rot=<0 0 -0.0124 -0.9999>}
 spine_02 {parent=spine_01, loc=<14.2907 -0.3558 0>, rot=<0 0 0.0159 -0.9999>}
 spine_03 {parent=spine_02, loc=<13.1068 0 0>, rot=<0 0 0.0041 -1>}
 neck_01 {parent=spine_03, loc=<2.5947 -1.5005 0>, rot=<0 0 0.4032 -0.9151>}
 neck_02 {parent=neck_01, loc=<5.711 0 0>, rot=<0 0 0.0462 -0.9989>}
 head {parent=neck_02, loc=<6.5839 0 0>, rot=<-0.2129 -0.6743 0.2129 -0.6743>}
 face_attach {parent=head, loc=<0 -0.8148 -0.4208>, rot=<0 0 0 1>}
 jaw {parent=face_attach, loc=<0 1.0725 1.4273>, rot=<0 0.7071 0 -0.7071>}
 lip_lower_03_l {parent=jaw, loc=<-4.0483 8.5878 -3.1509>, rot=<-0.0721 -0.6559 -0.2295 -0.7155>}
 lip_lower_02_l {parent=jaw, loc=<-4.3078 9.7759 -2.1948>, rot=<-0.2406 -0.6832 -0.3562 -0.5904>}
 lip_lower_01_l {parent=jaw, loc=<-4.4416 10.7595 -1.1753>, rot=<-0.3576 -0.704 -0.3942 -0.4702>}
 lip_lower_mid {parent=jaw, loc=<-4.2223 11.0052 0>, rot=<0.3881 0.5911 0.5911 0.3881>}
 lip_lower_01_r {parent=jaw, loc=<-4.4416 10.7595 1.1753>, rot=<0.704 -0.3576 -0.4702 0.3942>}
 lip_lower_02_r {parent=jaw, loc=<-4.3078 9.776 2.1948>, rot=<0.6832 -0.2406 -0.5904 0.3562>}
 lip_lower_03_r {parent=jaw, loc=<-4.0483 8.5878 3.1509>, rot=<-0.6559 0.0721 0.7155 -0.2295>}
 teeth_lower {parent=jaw, loc=<-4.1786 7.958 0>, rot=<0 -0.7071 0 -0.7071>}
 tongue_01 {parent=jaw, loc=<-1.2582 3.3853 0>, rot=<0.6017 -0.3714 0.3714 -0.6017>}
 tongue_02 {parent=tongue_01, loc=<-2.4201 0 0>, rot=<-0.0005 -0.062 0.0001 -0.9981>}
 tongue_03 {parent=tongue_02, loc=<-2.0927 0 0>, rot=<0.0005 -0.0094 -0.0001 -1>}
 lip_corner_l {parent=face_attach, loc=<4.2193 8.3992 -1.8726>, rot=<0.0176 -0.0132 -0.1977 -0.98>}
 cheek_bt_l {parent=face_attach, loc=<6.5575 5.747 -1.6239>, rot=<0.0809 -0.1726 -0.1547 -0.9694>}
 cheek_bt_r {parent=face_attach, loc=<-6.5575 5.7469 -1.6239>, rot=<0.9694 -0.1547 0.1726 0.0809>}
 reverse_jaw {parent=face_attach, loc=<0 0.0433 0.7402>, rot=<0 0.7071 0 -0.7071>}
 lip_upper_mid {parent=reverse_jaw, loc=<-1.6808 13.1404 0>, rot=<-0.4557 -0.5407 -0.5407 -0.4557>}
 lip_upper_01_l {parent=reverse_jaw, loc=<-2.0571 12.7201 -1.4511>, rot=<-0.4899 -0.6512 -0.3566 -0.457>}
 lip_upper_02_l {parent=reverse_jaw, loc=<-2.6632 11.7206 -2.8105>, rot=<-0.2406 -0.6832 -0.3562 -0.5904>}
 lip_upper_03_l {parent=reverse_jaw, loc=<-2.22 10.3784 -3.582>, rot=<-0.0167 -0.7319 -0.2603 -0.6295>}
 cheek_tp_in_l {parent=reverse_jaw, loc=<1.1042 11.3628 -2.3015>, rot=<-0.4855 -0.6882 -0.4984 -0.2056>}
 cheek_tp_ot_l {parent=reverse_jaw, loc=<1.3732 7.6666 -7.0875>, rot=<-0.1961 -0.6424 -0.2197 -0.7076>}
 cheek_tp_in_r {parent=reverse_jaw, loc=<1.1042 11.3628 2.3015>, rot=<0.6882 -0.4855 -0.2056 0.4984>}
 cheek_tp_ot_r {parent=reverse_jaw, loc=<1.3732 7.6666 7.0875>, rot=<-0.6424 0.1961 0.7076 -0.2197>}
 brow_ot_r {parent=reverse_jaw, loc=<5.3307 8.8034 6.2208>, rot=<-0.5915 -0.3152 0.633 0.3875>}
 brow_md_r {parent=reverse_jaw, loc=<6.1238 9.743 4.46>, rot=<0.7483 0.0798 -0.6522 -0.0916>}
 brow_ct {parent=reverse_jaw, loc=<5.7814 10.2512 0>, rot=<0 -0.7071 0 -0.7071>}
 brow_md_l {parent=reverse_jaw, loc=<6.1237 9.7429 -4.46>, rot=<0.0798 -0.7483 0.0916 -0.6522>}
 brow_in_l {parent=reverse_jaw, loc=<5.9564 9.9456 -2.1248>, rot=<0.1061 -0.6991 0.1061 -0.6991>}
 brow_ot_l {parent=reverse_jaw, loc=<5.3307 8.8034 -6.2208>, rot=<0.3152 -0.5915 0.3875 -0.633>}
 brow_in_r {parent=reverse_jaw, loc=<5.9564 9.9456 2.1248>, rot=<-0.6991 -0.1061 0.6991 0.1061>}
 eye_r {parent=reverse_jaw, loc=<3.7889 7.8018 3.6467>, rot=<-0.7012 -0.0583 0.708 0.0598>}
 lid_bt_r {parent=reverse_jaw, loc=<3.7883 7.778 3.6467>, rot=<-0.6966 -0.0705 0.7124 0.0476>}
 lid_tp_r {parent=reverse_jaw, loc=<3.7889 7.8018 3.6467>, rot=<-0.6966 -0.0705 0.7124 0.0476>}
 nose_mid {parent=reverse_jaw, loc=<0.4554 13.7896 0>, rot=<0.4878 0.5119 0.5119 0.4878>}
 nostril_l {parent=nose_mid, loc=<-0.0539 -0.8279 -0.434>, rot=<0 0 0 -1>}
 nostril_r {parent=nose_mid, loc=<-0.0539 0.8279 -0.4339>, rot=<0 -1 0 0>}
 lid_tp_l {parent=reverse_jaw, loc=<3.7889 7.8019 -3.6467>, rot=<0.0705 -0.6966 0.0476 -0.7124>}
 lid_bt_l {parent=reverse_jaw, loc=<3.7883 7.7781 -3.6467>, rot=<0.0705 -0.6966 0.0476 -0.7124>}
 eye_l {parent=reverse_jaw, loc=<3.7889 7.8019 -3.6467>, rot=<0.0583 -0.7012 0.0598 -0.708>}
 lip_upper_01_r {parent=reverse_jaw, loc=<-2.0571 12.7202 1.4512>, rot=<0.6512 -0.4899 -0.457 0.3566>}
 lip_upper_02_r {parent=reverse_jaw, loc=<-2.6633 11.7206 2.8105>, rot=<0.6832 -0.2406 -0.5904 0.3562>}
 lip_upper_03_r {parent=reverse_jaw, loc=<-2.22 10.3784 3.582>, rot=<0.7319 -0.0167 -0.6295 0.2603>}
 teeth_upper {parent=reverse_jaw, loc=<-0.8072 9.0408 0>, rot=<0 -0.7071 0 -0.7071>}
 ear_01_l {parent=reverse_jaw, loc=<1.5571 0.0575 -0.5535>, rot=<-0.0596 -0.2944 -0.0094 -0.9538>}
 ear_02_l {parent=ear_01_l, loc=<10.1531 0 0>, rot=<0.1107 -0.0026 -0.1045 -0.9883>}
 ear_03_l {parent=ear_02_l, loc=<2.3632 0 0>, rot=<0.0137 0.0739 -0.0375 -0.9965>}
 ear_04_l {parent=ear_03_l, loc=<2.608 0 0>, rot=<-0.0159 -0.0128 -0.0425 -0.9989>}
 ear_01_r {parent=reverse_jaw, loc=<1.5571 0.0575 0.5535>, rot=<-0.2944 0.0596 0.9538 -0.0094>}
 ear_02_r {parent=ear_01_r, loc=<-10.1531 0 0>, rot=<0.1107 -0.0026 -0.1045 -0.9883>}
 ear_03_r {parent=ear_02_r, loc=<-2.3632 -0.0001 0>, rot=<0.0137 0.0739 -0.0375 -0.9965>}
 ear_04_r {parent=ear_03_r, loc=<-2.6079 0.0001 -0.0001>, rot=<-0.0159 -0.0128 -0.0425 -0.9989>}
 lip_corner_r {parent=face_attach, loc=<-4.2193 8.3993 -1.8726>, rot=<0.98 -0.1977 0.0132 0.0176>}
 Crown {parent=neck_02, loc=<3.6895 9.0611 0>, rot=<-0.1467 -0.6917 0.1467 -0.6917>}
 Crown1 {parent=neck_02, loc=<-5.5122 12.8905 0>, rot=<-0.2213 -0.6716 0.2213 -0.6716>}
 shoulder_fr_l {parent=spine_03, loc=<1.9267 -0.6483 -5.7705>, rot=<0.0135 -0.0015 -0.7949 -0.6066>}
 elbow_fr_l {parent=shoulder_fr_l, loc=<10.8758 0 0>, rot=<-0.0045 0.004 0.1185 -0.9929>}
 wrist_fr_l {parent=elbow_fr_l, loc=<11.4815 0 0>, rot=<0.9986 0.0512 0.0161 -0.0013>}
 ball_fr_l {parent=wrist_fr_l, loc=<3.5887 0 0>, rot=<-0.0201 0.0298 -0.7317 -0.6807>}
 toe_fr_l {parent=ball_fr_l, loc=<3.0194 0 0>, rot=<-0.0262 0.0057 0.0732 -0.997>}
 toe_tip_fr_l {parent=toe_fr_l, loc=<2.9581 0 0>, rot=<0.0038 0.0063 -0.0812 -0.9967>}
 scapula_fr_l {parent=shoulder_fr_l, loc=<-5.0362 3.7723 -0.1058>, rot=<0.9487 -0.3158 0.0121 -0.0097>}
 shoulder_fr_r {parent=spine_03, loc=<1.9267 -0.6483 5.7706>, rot=<-0.0015 -0.0135 0.6066 -0.7949>}
 elbow_fr_r {parent=shoulder_fr_r, loc=<-10.8758 0 -0.0001>, rot=<-0.0045 0.004 0.1185 -0.9929>}
 wrist_fr_r {parent=elbow_fr_r, loc=<-11.4814 0 0>, rot=<0.9986 0.0512 0.0161 -0.0013>}
 ball_fr_r {parent=wrist_fr_r, loc=<-3.5887 0 0>, rot=<-0.0201 0.0298 -0.7317 -0.6807>}
 toe_fr_r {parent=ball_fr_r, loc=<-3.0195 0 0>, rot=<-0.0262 0.0057 0.0732 -0.997>}
 toe_tip_fr_r {parent=toe_fr_r, loc=<-2.958 0 0.0001>, rot=<0.0038 0.0063 -0.0812 -0.9967>}
 scapula_fr_r {parent=shoulder_fr_r, loc=<5.0362 -3.7723 0.1057>, rot=<0.9487 -0.3158 0.0121 -0.0097>}
 hips {parent=pelvis, loc=<0 0 0>, rot=<0 0 0 1>}
 thigh_bk_l {parent=hips, loc=<-1.7864 -5.2948 -5.6956>, rot=<0.7238 0.6899 -0.0073 0.0076>}
 knee_bk_l {parent=thigh_bk_l, loc=<12.276 0 0>, rot=<0.0064 -0.005 0.4145 -0.91>}
 ankle_bk_l {parent=knee_bk_l, loc=<11.9381 0 0>, rot=<-0.0046 0.0019 -0.5264 -0.8502>}
 ball_bk_l {parent=ankle_bk_l, loc=<10.3665 0 0>, rot=<-0.0054 -0.0039 -0.5826 -0.8127>}
 toe_bk_l {parent=ball_bk_l, loc=<1.8504 0 0>, rot=<0 0 0.1014 -0.9948>}
 toe_tip_bk_l {parent=toe_bk_l, loc=<2.362 0 0>, rot=<0 0 -0.1605 -0.987>}
 thigh_bk_r {parent=hips, loc=<-1.7864 -5.2948 5.6956>, rot=<0.6899 -0.7238 -0.0076 -0.0073>}
 knee_bk_r {parent=thigh_bk_r, loc=<-12.276 0 0>, rot=<0.0064 -0.005 0.4145 -0.91>}
 ankle_bk_r {parent=knee_bk_r, loc=<-11.9381 0 0>, rot=<-0.0046 0.0019 -0.5264 -0.8502>}
 ball_bk_r {parent=ankle_bk_r, loc=<-10.3664 0 0>, rot=<-0.0054 -0.0038 -0.5826 -0.8127>}
 toe_bk_r {parent=ball_bk_r, loc=<-1.8504 0 0>, rot=<0 0 0.1014 -0.9948>}
 toe_tip_bk_r {parent=toe_bk_r, loc=<-2.362 0 0>, rot=<0 0 -0.1605 -0.987>}
 tail_01 {parent=pelvis, loc=<-8.3781 -5.5168 0>, rot=<0.5008 0.4992 -0.5008 0.4992>}
 tail_02 {parent=tail_01, loc=<0 -6.652 0.2355>, rot=<0 0 0 1>}
 tail_03 {parent=tail_02, loc=<0 -6.6542 0.1637>, rot=<0 0 0 1>}
 tail_04 {parent=tail_03, loc=<0 -6.6562 -0.0241>, rot=<0 0 0 1>}
 tail_05 {parent=tail_04, loc=<0 -6.6483 0.286>, rot=<0 0 0 1>}
 IK_Foot_Rear_Root {parent=root, loc=<0 -19.464 0>, rot=<0 0 0 1>}
 IK_Foot_Rear_L {parent=IK_Foot_Rear_Root, loc=<6.2133 -2.0165 1.6094>, rot=<-0.4929 -0.507 -0.4929 -0.507>}
 IK_Foot_Rear_R {parent=IK_Foot_Rear_Root, loc=<-6.2133 -2.0165 1.6094>, rot=<0.507 -0.4929 0.507 -0.4929>}
 IK_Foot_Front_Root {parent=root, loc=<0 1.5578 0>, rot=<0 0 0 1>}
 IK_Foot_Front_L {parent=IK_Foot_Front_Root, loc=<6.1013 9.7225 1.8214>, rot=<-0.5039 -0.496 -0.4709 -0.5275>}
 IK_Foot_Front_R {parent=IK_Foot_Front_Root, loc=<-6.0973 9.6603 0.5459>, rot=<-0.5039 0.496 0.4709 -0.5275>}
 attach_emote_01 {parent=root, loc=<0 0 0>, rot=<0 0 0 1>}
 attach_emote_02 {parent=root, loc=<0 0 0>, rot=<0 0 0 1>}
 attach_emote_03 {parent=root, loc=<0 0 0>, rot=<0 0 0 1>}
 BothLegIK {type=Curve, loc=<0 0 0>, rot=<0 0 0 1>}
 L_eye_look_up_pose {type=Curve, loc=<0 0 0>, rot=<0 0 0 1>}
 L_wide_pose {type=Curve, loc=<0 0 0>, rot=<0 0 0 1>}
 R_lip_corner_narrow_pose {type=Curve, loc=<0 0 0>, rot=<0 0 0 1>}
 R_upper_lip_lower_pose {type=Curve, loc=<0 0 0>, rot=<0 0 0 1>}
 L_lip_corner_narrow_pose {type=Curve, loc=<0 0 0>, rot=<0 0 0 1>}
 L_upper_lip_lower_pose {type=Curve, loc=<0 0 0>, rot=<0 0 0 1>}
 R_lower_lip_up_pose {type=Curve, loc=<0 0 0>, rot=<0 0 0 1>}
 R_brow_squeeze_in_pose {type=Curve, loc=<0 0 0>, rot=<0 0 0 1>}
 L_lip_corner_wide_pose {type=Curve, loc=<0 0 0>, rot=<0 0 0 1>}
 L_upperLidDown_pose {type=Curve, loc=<0 0 0>, rot=<0 0 0 1>}
 R_lower_lip_down_pose {type=Curve, loc=<0 0 0>, rot=<0 0 0 1>}
 R_upper_lip_raiser_pose {type=Curve, loc=<0 0 0>, rot=<0 0 0 1>}
 L_lower_lip_down_pose {type=Curve, loc=<0 0 0>, rot=<0 0 0 1>}
 L_upper_lip_raiser_pose {type=Curve, loc=<0 0 0>, rot=<0 0 0 1>}
 L_brow_squeeze_in_pose {type=Curve, loc=<0 0 0>, rot=<0 0 0 1>}
 L_lower_lip_up_pose {type=Curve, loc=<0 0 0>, rot=<0 0 0 1>}
 R_upperLidDown_pose {type=Curve, loc=<0 0 0>, rot=<0 0 0 1>}
 R_wide_pose {type=Curve, loc=<0 0 0>, rot=<0 0 0 1>}
 R_blink_pose {type=Curve, loc=<0 0 0>, rot=<0 0 0 1>}
 L_eye_look_left_pose {type=Curve, loc=<0 0 0>, rot=<0 0 0 1>}
 L_eye_look_down_pose {type=Curve, loc=<0 0 0>, rot=<0 0 0 1>}
 L_blink_pose {type=Curve, loc=<0 0 0>, rot=<0 0 0 1>}
 Reaction_Happy {type=Curve, loc=<0 0 0>, rot=<0 0 0 1>}
 L_eye_look_right_pose {type=Curve, loc=<0 0 0>, rot=<0 0 0 1>}
 R_lip_corner_wide_pose {type=Curve, loc=<0 0 0>, rot=<0 0 0 1>}
 L_smile_pose {type=Curve, loc=<0 0 0>, rot=<0 0 0 1>}
 R_smile_pose {type=Curve, loc=<0 0 0>, rot=<0 0 0 1>}
 L_frown_pose {type=Curve, loc=<0 0 0>, rot=<0 0 0 1>}
 R_frown_pose {type=Curve, loc=<0 0 0>, rot=<0 0 0 1>}
 C_nose_wrinkler_pose {type=Curve, loc=<0 0 0>, rot=<0 0 0 1>}
 L_cheek_up_pose {type=Curve, loc=<0 0 0>, rot=<0 0 0 1>}
 R_cheek_up_pose {type=Curve, loc=<0 0 0>, rot=<0 0 0 1>}
 jaw_open_pose {type=Curve, loc=<0 0 0>, rot=<0 0 0 1>}
 jaw_left_pose {type=Curve, loc=<0 0 0>, rot=<0 0 0 1>}
 jaw_right_pose {type=Curve, loc=<0 0 0>, rot=<0 0 0 1>}
 jaw_up_pose {type=Curve, loc=<0 0 0>, rot=<0 0 0 1>}
 jaw_forward_pose {type=Curve, loc=<0 0 0>, rot=<0 0 0 1>}
 jaw_back_pose {type=Curve, loc=<0 0 0>, rot=<0 0 0 1>}
 eye_convergence_pose {type=Curve, loc=<0 0 0>, rot=<0 0 0 1>}
 tongue_forward_pose {type=Curve, loc=<0 0 0>, rot=<0 0 0 1>}
 tongue_back_pose {type=Curve, loc=<0 0 0>, rot=<0 0 0 1>}
 tongue_up_pose {type=Curve, loc=<0 0 0>, rot=<0 0 0 1>}
 tongue_down_pose {type=Curve, loc=<0 0 0>, rot=<0 0 0 1>}
 L_brow_In_Up_pose {type=Curve, loc=<0 0 0>, rot=<0 0 0 1>}
 L_brow_Ot_Up_pose {type=Curve, loc=<0 0 0>, rot=<0 0 0 1>}
 L_brow_In_Dn_pose {type=Curve, loc=<0 0 0>, rot=<0 0 0 1>}
 L_brow_Ot_Dn_pose {type=Curve, loc=<0 0 0>, rot=<0 0 0 1>}
 R_brow_In_Up_pose {type=Curve, loc=<0 0 0>, rot=<0 0 0 1>}
 R_brow_Ot_Up_pose {type=Curve, loc=<0 0 0>, rot=<0 0 0 1>}
 R_brow_In_Dn_pose {type=Curve, loc=<0 0 0>, rot=<0 0 0 1>}
 R_brow_Ot_Dn_pose {type=Curve, loc=<0 0 0>, rot=<0 0 0 1>}
 upper_lip_up_pose {type=Curve, loc=<0 0 0>, rot=<0 0 0 1>}
 upper_lip_dn_pose {type=Curve, loc=<0 0 0>, rot=<0 0 0 1>}
 lower_lip_up_pose {type=Curve, loc=<0 0 0>, rot=<0 0 0 1>}
 lower_lip_dn_pose {type=Curve, loc=<0 0 0>, rot=<0 0 0 1>}
 L_smile_open_correction_pose {type=Curve, loc=<0 0 0>, rot=<0 0 0 1>}
 R_smile_open_correction_pose {type=Curve, loc=<0 0 0>, rot=<0 0 0 1>}
 tongue_left_pose {type=Curve, loc=<0 0 0>, rot=<0 0 0 1>}
 tongue_right_pose {type=Curve, loc=<0 0 0>, rot=<0 0 0 1>}
 reverse_jaw_pose {type=Curve, loc=<0 0 0>, rot=<0 0 0 1>}
 base_pose {type=Curve, loc=<0 0 0>, rot=<0 0 0 1>}
 R_eye_look_left_pose {type=Curve, loc=<0 0 0>, rot=<0 0 0 1>}
 R_eye_look_right_pose {type=Curve, loc=<0 0 0>, rot=<0 0 0 1>}
 R_eye_look_up_pose {type=Curve, loc=<0 0 0>, rot=<0 0 0 1>}
 R_eye_look_down_pose {type=Curve, loc=<0 0 0>, rot=<0 0 0 1>}
 L_lower_lid_up_pose {type=Curve, loc=<0 0 0>, rot=<0 0 0 1>}
 R_lower_lid_up_pose {type=Curve, loc=<0 0 0>, rot=<0 0 0 1>}
 tongue_curl_tip_up_pose {type=Curve, loc=<0 0 0>, rot=<0 0 0 1>}
 tongue_curl_tip_down_pose {type=Curve, loc=<0 0 0>, rot=<0 0 0 1>}
 tongue_curl_mid_up_pose {type=Curve, loc=<0 0 0>, rot=<0 0 0 1>}
 tongue_curl_mid_down_pose {type=Curve, loc=<0 0 0>, rot=<0 0 0 1>}
 L_tongue_twist_pose {type=Curve, loc=<0 0 0>, rot=<0 0 0 1>}
 R_tongue_twist_pose {type=Curve, loc=<0 0 0>, rot=<0 0 0 1>}
 AttachEmote1InPelvisSpace {type=Curve, loc=<0 0 0>, rot=<0 0 0 1>}
 AttachEmote2InPelvisSpace {type=Curve, loc=<0 0 0>, rot=<0 0 0 1>}
 AttachEmote3InPelvisSpace {type=Curve, loc=<0 0 0>, rot=<0 0 0 1>}
 AnimateAcc {type=Curve, loc=<0 0 0>, rot=<0 0 0 1>}
 pupil_shape {type=Curve, loc=<0 0 0>, rot=<0 0 0 1>}
 pupil_size {type=Curve, loc=<0 0 0>, rot=<0 0 0 1>}
 bypass_poseAsset {type=Curve, loc=<0 0 0>, rot=<0 0 0 1>}
 Hide Element 01 {type=Curve, loc=<0 0 0>, rot=<0 0 0 1>}
 Hide Element 02 {type=Curve, loc=<0 0 0>, rot=<0 0 0 1>}
]
