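# Post-process AnimBP for the CarrierShade quadruped cosmetic companion, correcting its Crown, Crown1 and emote attach bones.
# The input pose passes through a curve-driven pose asset, then weighted multi-bone parent constraints position both crown bones from neck, head, face and spine bones.
# Crown is offset and aimed at Crown1, and attach_emote_01 copies Glasses_Aim's translation.
# Spare constraint, Glasses_Aim look-at and pose nodes never reach the output; the EventGraph is empty.
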AnimBlueprint /CosmeticCompanions/Assets/Quadruped/CarrierShade/Meshes/CarrierShade_Quadruped_Post_AnimBP
target_skeleton=&/CarrierShade_Quadruped_Skeleton
graphs [
 AnimGraph {nodes [
   n1 = "Constraint - Crown (P)"!(ComponentPose, Alpha=1, ConstraintWeights_0=1, ConstraintWeights_1:float=0) -> (Pose) {boneToModify {boneName=Crown}, constraintSetup [{targetBone {boneName=Crown}}], constraintWeights [1,0]}
   "Constraint - Crown1 (PPPPPPP)"!(
    ComponentPose=@n1
    Alpha=1
    ConstraintWeights_0=0.5026
    ConstraintWeights_1:float=0.1851
    ConstraintWeights_2:float=0.1662
    ConstraintWeights_3:float=0.0469
    ConstraintWeights_4:float=0.0467
    ConstraintWeights_5:float=0.0393
    ConstraintWeights_6:float=0.0131
   ) -> (Pose) {
    boneToModify {boneName=Crown1}
    constraintSetup [
     {targetBone {boneName=neck_01}}
     {targetBone {boneName=spine_03}}
     {targetBone {boneName=neck_02}}
     {targetBone {boneName=shoulder_fr_r}}
     {targetBone {boneName=shoulder_fr_l}}
     {targetBone {boneName=spine_02}}
     {targetBone {boneName=face_attach}}
    ]
    constraintWeights [1,1,1,1,1,1,1]
   }
   Look At - Bone: Glasses_Aim(Alpha=1) {
    boneToModify {boneName=Glasses_Aim}
    lookAtLocation {x=0, z=-100}
    lookAtTarget {boneReference {boneName=nose_mid}}
    lookAt_Axis {axis {y=0, z=-1}}
    lookUp_Axis {axis=<0 1>}
   }
   Pose!(PoseAsset:Pose Asset Object Reference=Get PoseAsset() -> (PoseAsset:Pose Asset Object Reference), SourcePose:Pose Link Structure) -> (Pose:Pose Link Structure) {is=title AnimGraphNode_PoseBlendNode, bIgnorePoseAssetSourceAnimationCurves=true, blendOption=Linear}
   n2 = Input Pose()
   n3 = "Constraint - Crown (PPPPPPPP)"!(
    ComponentPose=Local To Component(LocalPose=CarrierShade_Quadruped_PoseAsset!(SourcePose:Pose Link Structure=@n2) -> (Pose:Pose Link Structure) {is=title AnimGraphNode_PoseBlendNode, poseAsset=&/CarrierShade_Quadruped_PoseAsset})
    Alpha=1
    ConstraintWeights_0=0.4111
    ConstraintWeights_1:float=0.1877
    ConstraintWeights_2:float=0.1775
    ConstraintWeights_3:float=0.1602
    ConstraintWeights_4:float=0.0307
    ConstraintWeights_5:float=0.0269
    ConstraintWeights_6:float=0.0058
    ConstraintWeights_7:float=0.0001
   ) -> (Pose) {
    boneToModify {boneName=Crown}
    constraintSetup [
     {targetBone {boneName=neck_02}}
     {targetBone {boneName=neck_01}}
     {targetBone {boneName=jaw}}
     {targetBone {boneName=head}}
     {targetBone {boneName=cheek_bt_r}}
     {targetBone {boneName=cheek_bt_l}}
     {targetBone {boneName=spine_03}}
     {targetBone {boneName=lip_lower_03_l}}
    ]
    constraintWeights [1,1,1,1,1,1,1,1]
   }
   n4 = "Constraint - Crown1 (PPPPPPP)"(
    ComponentPose=@n3
    Alpha=1
    ConstraintWeights_0=0.5026
    ConstraintWeights_1=0.1851
    ConstraintWeights_2=0.1662
    ConstraintWeights_3=0.0469
    ConstraintWeights_4=0.0467
    ConstraintWeights_5=0.0393
    ConstraintWeights_6=0.0131
   ) {
    boneToModify {boneName=Crown1}
    constraintSetup [
     {targetBone {boneName=neck_01}}
     {targetBone {boneName=spine_03}}
     {targetBone {boneName=neck_02}}
     {targetBone {boneName=shoulder_fr_r}}
     {targetBone {boneName=shoulder_fr_l}}
     {targetBone {boneName=spine_02}}
     {targetBone {boneName=face_attach}}
    ]
    constraintWeights [1,1,1,1,1,1,1]
   }
   n5 = "Transform (Modify) Bone - Bone: Crown"(ComponentPose=@n4, Alpha=1, Translation="-0.000,0.371,0.331", Rotation="0,0,0", Scale="1,1,1") {translationMode=BMM_Additive, translationSpace=BCS_BoneSpace, boneToModify {boneName=Crown}, translation {y=0.371, z=0.331}}
   n6 = Look At - Bone: Crown(ComponentPose=@n5, Alpha=1) {
    boneToModify {boneName=Crown}
    lookAtLocation=<0 3>
    lookAtTarget {boneReference {boneName=Crown1}}
    lookAt_Axis {axis {y=0, z=-1}}
    lookUp_Axis {axis=<0 1>}
   }
   n7 = "Constraint - attach_emote_01 (T)"(
    ComponentPose="Transform (Modify) Bone - Bone: Crown1"(ComponentPose=@n6, Alpha=1, Translation="0,0,0", Rotation="0,0,0", Scale="1,1,1") {translationMode=BMM_Additive, translationSpace=BCS_BoneSpace, boneToModify {boneName=Crown1}}
    Alpha=1
    ConstraintWeights_0=1
   ) {
    boneToModify {boneName=attach_emote_01}
    constraintSetup [{transformType=Translation, targetBone {boneName=Glasses_Aim}}]
    constraintWeights [1]
   }
   Output Pose(Result=Component To Local(ComponentPose=@n7))
  ]}
 EventGraph {links [], nodes []}
]
variables [PoseAsset {type=PoseAsset, default=&CarrierShade_Quadruped_PoseAsset}]
